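# Physics asset for the Horned Mask pickaxe mesh, giving the cosmetic melee weapon capsule bodies for its handle, blade and dangling parts.
# The blade is rotationally locked to the body and anchors a three-bone tuft chain and two tassel chains of base, top and bottom bones.
# Tassel joints lock selected axes, so each base only twists, each top swings on one axis, and each bottom both swings and twists.
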PhysicsAsset /BRCosmetics/Weapons/FORT_Melee/Pickaxe_HornedMask/Meshes/Horned_Mask_Pickaxe_Mesh_Physics
bodies {
 blade [
  {shapeType=Capsule, length=15, radius=8.505, shapeName=collider1, center=<0 0 -10.4059>, rotation=<rot 0 -90 0>}
  {shapeType=Capsule, length=42.75, radius=8.5, shapeName=collider2, rotation=<rot 0 0 -90>}
 ]
 body [{shapeType=Capsule, length=116.7737, radius=2.5265, center=<0.72 -0.9613 27.5064>, rotation=<rot -0.0288 -89.9995 -0.0802>}]
 dyn_tassleA_base [{shapeType=Capsule, length=0.505, radius=0.505, rotation=<rot 0 -90 0>}]
 dyn_tassleA_bot [{shapeType=Capsule, length=0.0154, radius=3.1629, center=<0.04 -0.1123 -2.972>, rotation=<rot 1.3983 -97.2248 -90>}]
 dyn_tassleA_top [{shapeType=Capsule, length=0.1595, radius=1.2097, center=<0.0201 0.1922 -1.9712>, rotation=<rot 0.2554 -90 4.2474>}]
 dyn_tassleB_base [{shapeType=Capsule, length=0.505, radius=0.505, rotation=<rot 0 -90 0>}]
 dyn_tassleB_bot [{shapeType=Capsule, length=0.2764, radius=3.1346, center=<0.0181 0.1109 -3.4063>, rotation=<rot 0 -155.592 -87.2794>}]
 dyn_tassleB_top [{shapeType=Capsule, length=0.1521, radius=1.1474, center=<-0.0193 0.2142 -1.9766>, rotation=<rot 3.77 -90 -3.1243>}]
 dyn_tuft_bot [{shapeType=Capsule, length=1.8226, radius=1.0243, center=<4.1697 -0.0639 -0.7383>, rotation=<rot 0 99.8272 -63.6612>}]
 dyn_tuft_mid [{shapeType=Capsule, length=3.7958, radius=2.7359, center=<5.6196 0.0171 -1.2648>, rotation=<rot -0.7762 -90 24.224>}]
 dyn_tuft_top [{shapeType=Capsule, length=4.3092, radius=3.3619, center=<4.7774 -0.2435 -0.6165>, rotation=<rot 22.3756 -90 -1.5788>}]
}
constraints [
 {bone1Name=dyn_tuft_top, bone2Name=blade, swing1LimitDegrees=45, swing2LimitDegrees=45, twistLimitDegrees=45}
 {bone1Name=dyn_tuft_mid, bone2Name=dyn_tuft_top, swing1LimitDegrees=45, swing2LimitDegrees=45, twistLimitDegrees=45}
 {bone1Name=dyn_tuft_bot, bone2Name=dyn_tuft_mid, swing1LimitDegrees=45, swing2LimitDegrees=45, twistLimitDegrees=45}
 {
  bone1Name=blade
  bone2Name=body
  swing1LimitDegrees=45
  swing1Motion=Locked
  swing2LimitDegrees=45
  swing2Motion=Locked
  twistLimitDegrees=45
  twistMotion=Locked
 }
 {
  bone1Name=dyn_tassleA_top
  bone2Name=dyn_tassleA_base
  swing1LimitDegrees=45
  swing1Motion=Locked
  swing2LimitDegrees=45
  twistLimitDegrees=45
  twistMotion=Locked
 }
 {
  bone1Name=dyn_tassleA_bot
  bone2Name=dyn_tassleA_top
  swing1LimitDegrees=45
  swing1Motion=Locked
  swing2LimitDegrees=45
  twistLimitDegrees=15
 }
 {
  bone1Name=dyn_tassleB_top
  bone2Name=dyn_tassleB_base
  swing1LimitDegrees=45
  swing1Motion=Locked
  swing2LimitDegrees=45
  twistLimitDegrees=45
  twistMotion=Locked
 }
 {
  bone1Name=dyn_tassleB_bot
  bone2Name=dyn_tassleB_top
  swing1LimitDegrees=45
  swing1Motion=Locked
  swing2LimitDegrees=45
  twistLimitDegrees=15
 }
 {
  bone1Name=dyn_tassleB_base
  bone2Name=blade
  swing1LimitDegrees=45
  swing1Motion=Locked
  swing2LimitDegrees=45
  swing2Motion=Locked
  twistLimitDegrees=22
 }
 {
  bone1Name=dyn_tassleA_base
  bone2Name=blade
  swing1LimitDegrees=45
  swing1Motion=Locked
  swing2LimitDegrees=45
  swing2Motion=Locked
  twistLimitDegrees=22
 }
]
