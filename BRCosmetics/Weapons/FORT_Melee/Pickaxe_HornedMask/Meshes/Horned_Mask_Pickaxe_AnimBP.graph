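# Animation Blueprint for the Horned Mask pickaxe melee weapon, adding procedural secondary motion to its skeleton.
# The input pose feeds a RigidBody node using the pickaxe's physics asset with scaled-down component motion and overridden gravity, then a Trail controller on the tuft bone chain before output.
# A four-node Anim Dynamics chain for the two tassels is left disconnected from the output.
# The event graph holds no logic.
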AnimBlueprint /BRCosmetics/Weapons/FORT_Melee/Pickaxe_HornedMask/Meshes/Horned_Mask_Pickaxe_AnimBP
target_skeleton=&/Horned_Mask_Pickaxe_Skeleton
graphs [
 AnimGraph {nodes [
   n1 = Anim Dynamics - Bone: dyn_tassleA_top(Alpha=1, ExternalForce="0,0,0") {
    angularDampingOverride=0.999
    bGravityOverrideInSimSpace=true
    bOverrideAngularDamping=true
    bUseGravityOverride=true
    alphaBoolBlend {blendOption=Linear}
    boundBone {boneName=dyn_tassleA_top}
    componentAppliedLinearAccClamp=<300 300 300>
    componentLinearAccScale=<1 1 1>
    componentLinearVelScale=<1 1 1>
    gravityOverride {z=-981}
    physicsBodyDefinitions [{
      sphereCollisionRadius=0
      boundBone {boneName=dyn_tassleA_top}
      boxExtents=<16 16 4>
      constraintSetup {angularLimitsMax=<0 15 0>, angularLimitsMin=<0 -15 0>, angularTarget=<0 0 0>}
      localJointOffset=<0 0 -5>
     }]
   }
   n2 = Anim Dynamics - Bone: dyn_tassleA_bot(ComponentPose=@n1, Alpha=1, ExternalForce="0,0,0") {
    angularDampingOverride=0.999
    bGravityOverrideInSimSpace=true
    bOverrideAngularDamping=true
    bUseGravityOverride=true
    alphaBoolBlend {blendOption=Linear}
    boundBone {boneName=dyn_tassleA_bot}
    componentAppliedLinearAccClamp=<50 50 50>
    componentLinearAccScale=<1 1 1>
    componentLinearVelScale=<1 1 1>
    gravityOverride {z=-981}
    physicsBodyDefinitions [{
      sphereCollisionRadius=0
      boundBone {boneName=dyn_tassleA_bot}
      boxExtents=<4 4 1>
      constraintSetup {angularLimitsMax=<10 10 0>, angularLimitsMin=<-10 -10 0>, angularTarget=<0 0 0>}
      localJointOffset=<0 0 -5>
     }]
   }
   n3 = Anim Dynamics - Bone: dyn_tassleB_top(ComponentPose=@n2, Alpha=1, ExternalForce="0,0,0") {
    angularDampingOverride=0.999
    bGravityOverrideInSimSpace=true
    bOverrideAngularDamping=true
    bUseGravityOverride=true
    alphaBoolBlend {blendOption=Linear}
    boundBone {boneName=dyn_tassleB_top}
    componentAppliedLinearAccClamp=<300 300 300>
    componentLinearAccScale=<1 1 1>
    componentLinearVelScale=<1 1 1>
    gravityOverride {z=-981}
    physicsBodyDefinitions [{
      sphereCollisionRadius=0
      boundBone {boneName=dyn_tassleB_top}
      boxExtents=<16 16 4>
      constraintSetup {angularLimitsMax=<0 15 0>, angularLimitsMin=<0 -15 0>, angularTarget=<0 0 0>}
      localJointOffset=<0 0 -5>
     }]
   }
   Anim Dynamics - Bone: dyn_tassleB_bot(ComponentPose=@n3, Alpha=1, ExternalForce="0,0,0") {
    angularDampingOverride=0.999
    bGravityOverrideInSimSpace=true
    bOverrideAngularDamping=true
    bUseGravityOverride=true
    alphaBoolBlend {blendOption=Linear}
    boundBone {boneName=dyn_tassleB_bot}
    componentAppliedLinearAccClamp=<50 50 50>
    componentLinearAccScale=<1 1 1>
    componentLinearVelScale=<1 1 1>
    gravityOverride {z=-981}
    physicsBodyDefinitions [{
      sphereCollisionRadius=0
      boundBone {boneName=dyn_tassleB_bot}
      boxExtents=<4 4 1>
      constraintSetup {angularLimitsMax=<10 10 0>, angularLimitsMin=<-10 -10 0>, angularTarget=<0 0 0>}
      localJointOffset=<0 0 -5>
     }]
   }
   AnimGraphNode_SubInput_0 = Input Pose()
   n4 = RigidBody(ComponentPose=Local To Component(LocalPose=@AnimGraphNode_SubInput_0)) {
    bOverrideWorldGravity=true
    overridePhysicsAsset=&/Horned_Mask_Pickaxe_Mesh_Physics
    alphaBoolBlend {blendOption=Linear}
    componentAppliedLinearAccClamp=<300 300 300>
    componentLinearAccScale=<0.35 0.35 0.35>
    componentLinearVelScale=<0.4 0.4 0.4>
    overrideWorldGravity {z=-980}
   }
   n5 = Trail controller - Bone: dyn_tuft_bot(ComponentPose=@n4, Alpha=1) {
    chainLength=3
    relaxationSpeedScale=3
    rotationLimits [{limitMax=<180 180 180>}, {limitMax=<180 180 180>}, {limitMax=<180 180 180>}]
    rotationOffsets [<0 0 0>, <0 0 0>, <0 0 0>]
    trailBone {boneName=dyn_tuft_bot}
   }
   Output Pose(Result=Component To Local(ComponentPose=@n5)) {layerGroup=None}
  ]}
 EventGraph {nodes [Event BlueprintUpdateAnimation()]}
]
variables []
